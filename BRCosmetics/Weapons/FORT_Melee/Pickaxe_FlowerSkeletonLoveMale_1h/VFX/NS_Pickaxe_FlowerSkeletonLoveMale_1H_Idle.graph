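# Niagara idle effect for the FlowerSkeletonLoveMale one-handed pickaxe cosmetic, shedding sparse petal sprites.
# A single Petals emitter spawns probabilistically in a small sphere, with random velocity plus speed-capped inherited motion.
# Per-particle gravity, drag, curl noise and spin make petals drift downward.
# A sine on emitter age, with its period varied by material random, oscillates sprite width to fake flipping, while a curve shrinks petals at end of life.
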NiagaraSystem /BRCosmetics/Weapons/FORT_Melee/Pickaxe_FlowerSkeletonLoveMale_1h/VFX/NS_Pickaxe_FlowerSkeletonLoveMale_1H_Idle
dependencies {usedDataInterfaces [&NiagaraDataInterfaceCurve], usedDynamicInputs [], usedModules [&SystemState]}
emitters {Petals {
  emitterUpdateScript [
   &EmitterState(Life Cycle Mode=::NewEnumerator1, Loop Duration=0.625, UseLoopDelay=false)
   &SpawnRate(Spawn Probability=0.1, SpawnRate=4, Use Spawn Probability=true)
  ]
  particleSpawnScript [
   /Niagara/Modules/Spawn/Initialization/InitializeParticle(
    Color=<rgba 0 0 0 1>
    Lifetime=2.5
    Module.Write Color=true
    Module.Write Lifetime=true
    Module.Write Mass=true
    Module.Write Material Random=true
    Module.Write Position=false
    Module.Write Scale=false
    Module.Write SpriteRotation=true
    Module.Write SpriteSize=true
    Position=&SimulationPosition()
    Sprite Rotation=&UniformRangedFloat(Minimum=0, Maximum=360)
    Sprite Size=&Multiply_Vector2(A=<2.5 2.5>, B=&UniformRangedVector2D(Minimum=<1 0.5>, Maximum=<2 2>))
   )
   &SphereLocation(
    MaskSpawnGroup=false
    Override Seed=false
    Sphere Radius=3
    Surface Only=false
    UseNonUniformScale=false
    UseOffset=false
    UseSphereOrientation=false
   )
   &AddVelocity(Velocity=&VectorFromFloat(Value=&UniformRangedFloat(Minimum=-10, Maximum=10)))
   &InheritVelocity(Module.Limit Speed=true, Velocity Limit=50)
   /SetVariables(Particles.stableGravity:NiagaraFloat=&UniformRangedFloat(Minimum=-20, Maximum=-35)) {isSetParameters=true}
   /SetVariables(Particles.stableRotationRate:NiagaraFloat=&UniformRangedFloat(Minimum=50, Maximum=200)) {isSetParameters=true}
  ]
  particleUpdateScript [
   &ParticleState()
   &GravityForce(Gravity=&MakeVector(X=0, Y=0, Z=$Particles.Initial.stableGravity))
   &Drag(Drag=&UniformRangedFloat(Minimum=0.6, Maximum=0.9), Use Linear Drag=true, Use Rotational Drag=false)
   /Niagara/Modules/Update/Forces/CurlNoiseForce(Noise Strength=50, Pan Noise=false)
   &SolveForcesAndVelocity(Clamp Velocity=false, Limit Acceleration=false)
   &SpriteRotationRate(Rotation Rate=$Particles.Initial.stableRotationRate)
   &ScaleSpriteSize(Scale Factor=&Multiply_Vector2(
     A=&MakeVector2D(
      X=&Multiply_Float(
       A=1
       B=&Sine(
        Normalized Angle=$Emitter.Age
        Period=&Multiply_Float(A=1.35, B=&Add_Float(A=$Particles.Initial.MaterialRandom, B=0.6))
        Scale=1
        Bias=0
       )
      )
      Y=1
     )
     B=&Vector2DFromFloat(Value=&FloatFromCurve(FloatCurve=DataInterface() {propertyValues {Curve {keys [<at 0 1.1>,<at 0.9 1.1>,<at 1 0>]}}}))
    ))
  ]
  rendererClasses [&NiagaraSpriteRendererProperties]
  renderers [{rendererClass=&NiagaraSpriteRendererProperties}]
 }}
summary {
 systemName=NS_Pickaxe_FlowerSkeletonLoveMale_1H_Idle
 emitters [{emitterName=Petals, rendererClasses [&NiagaraSpriteRendererProperties]}]
 userVariables [
  User.bIsFrontEndPreview {type=NiagaraBool}
  User.PlayerSpeed {type=NiagaraFloat, defaultValue=0, description=""}
  User.SkeletalMesh {
   type=NiagaraDataInterfaceSkeletalMesh
   defaultValue {
    struct=NiagaraExt_VariableValue_DataInterface
    value {dataInterface=&self:NiagaraDataInterfaceSkeletalMesh_1}
   }
  }
 ]
}
