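# Niagara swing effect for the one-handed FlowerSkeletonLoveMale pickaxe, combining petals, sparkles and a weapon trail.
# Petals inherit velocity and fall with curl noise, faking flutter through sine-driven sprite width.
# Sparkles spawn per distance moved, scattered along and slightly beyond the line between two sampled skeletal mesh points, and twinkle via sine scaling.
# Trail is a short-lived magenta AnimTrail ribbon, alpha-shaped by execution index and scaled by time of day.
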NiagaraSystem /BRCosmetics/Weapons/FORT_Melee/Pickaxe_FlowerSkeletonLoveMale_1h/VFX/NS_Pickaxe_FlowerSkeletonLoveMale_1H_Swing
dependencies {
 usedDataInterfaces [&NiagaraDataInterfaceMeshRendererInfo, &NiagaraDataInterfaceCurve]
 usedDynamicInputs []
 usedModules [&SystemState]
}
emitters {
 Petals {
  emitterUpdateScript [
   &EmitterState(Life Cycle Mode=::NewEnumerator1, Loop Behavior=::NewEnumerator1, Loop Duration=0.5, UseLoopDelay=false)
   &SpawnRate(SpawnRate=30, Use Spawn Probability=false)
  ]
  particleSpawnScript [
   &InitializeParticle(
    Color=<rgba 0 0 0 1>
    Lifetime Max=1
    Lifetime Min=0.5
    Lifetime Mode=::NewEnumerator1
    Sprite Rotation Mode=::NewEnumerator2
    Sprite Size Max=<5 5>
    Sprite Size Min=<2.5 1.25>
    Sprite Size Mode=::NewEnumerator2
    UsePositionOffset=false
    Write Material Random=true
   ) {module=InitializeParticle001}
   &SphereLocation(
    MaskSpawnGroup=false
    Override Seed=false
    Sphere Radius=15
    Surface Only=false
    UseNonUniformScale=false
    UseOffset=false
    UseSphereOrientation=false
   )
   &AddVelocity(Velocity=&UniformRangedVector(Minimum=<-20 -5 10>, Maximum=<20 -20 30>)) {enabled=false}
   &InheritVelocity(Module.Limit Speed=true)
   /SetVariables(Particles.stableGravity:NiagaraFloat=&UniformRangedFloat(Minimum=-150, Maximum=-200)) {isSetParameters=true}
   /SetVariables(Particles.stableRotationRate:NiagaraFloat=&UniformRangedFloat(Minimum=50, Maximum=200)) {isSetParameters=true}
  ]
  particleUpdateScript [
   &ParticleState()
   &GravityForce(Gravity=&MakeVector(X=0, Y=0, Z=$Particles.Initial.stableGravity))
   &Drag(Drag=&UniformRangedFloat(Minimum=1.2, Maximum=1.5), Use Linear Drag=true, Use Rotational Drag=false)
   /Niagara/Modules/Update/Forces/CurlNoiseForce(Noise Strength=300, Pan Noise=false)
   &SolveForcesAndVelocity(Clamp Velocity=false, Limit Acceleration=false)
   &SpriteRotationRate(Rotation Rate=$Particles.Initial.stableRotationRate)
   &ScaleSpriteSize(Scale Factor=&Multiply_Vector2(
     A=&MakeVector2D(
      X=&Multiply_Float(
       A=1
       B=&Sine(
        Normalized Angle=$Emitter.Age
        Period=&Multiply_Float(A=0.7, B=&Add_Float(A=$Particles.Initial.MaterialRandom, B=0.65))
        Scale=1
        Bias=0
       )
      )
      Y=1
     )
     B=&Vector2DFromFloat(Value=&FloatFromCurve(FloatCurve=DataInterface() {propertyValues {Curve {keys [<at 0 1.1>,<at 0.9 1.1>,<at 1 0>]}}}))
    ))
  ]
  rendererClasses [&NiagaraSpriteRendererProperties]
  renderers [{rendererClass=&NiagaraSpriteRendererProperties}]
 }
 Sparkles {
  emitterUpdateScript [
   &EmitterState(Life Cycle Mode=::NewEnumerator1, Loop Behavior=::NewEnumerator1, Loop Duration=0.66, UseLoopDelay=false)
   &SpawnPerUnit(Module.Use Max Movement Threshold=true, Module.Use Movement Tolerance=false, Spawn Spacing=16, Use Spawn Probability=false)
   &SpawnRate(SpawnRate=200, Use Spawn Probability=false) {enabled=false}
  ]
  particleSpawnScript [
   &SampleSkeletalMesh(
    AdditionalYawPitchRoll=false
    Mesh Sampling Type:ENiagaraSkelSamplingModeFull=::NewEnumerator5
    Skeletal Mesh:NiagaraDataInterfaceSkeletalMesh=$User.SkeletalMesh
    Socket Sampling Mode:ENiagaraSocketSamplingMode=::NewEnumerator5
   )
   &SampleSkeletalMesh(
    AdditionalYawPitchRoll=false
    Bone / Socket Index=1
    Mesh Sampling Type:ENiagaraSkelSamplingModeFull=::NewEnumerator5
    Skeletal Mesh:NiagaraDataInterfaceSkeletalMesh=$User.SkeletalMesh
    Socket Sampling Mode:ENiagaraSocketSamplingMode=::NewEnumerator5
   ) {module=SampleSkeletalMesh001}
   &InitializeParticle(
    Color Maximum=<rgba 4 1 2 1>
    Color Minimum=<rgba 1 0.1 0.4 1>
    Color Mode=::NewEnumerator2
    Lifetime Max=0.33
    Lifetime Min=0.1
    Lifetime Mode=::NewEnumerator1
    Position=&Lerp_Vector(
     A=$Particles.SampleSkeletalMesh.SampledPosition
     B=$Particles.SampleSkeletalMesh001.SampledPosition
     Alpha=&RandomRangeFloat(Minimum=-0.5, Maximum=1.5)
    )
    Position Mode=::NewEnumerator1
    Sprite Size Mode=::NewEnumerator1
    Uniform Sprite Size Max=8
    UsePositionOffset=false
    Write Material Random=true
   )
   &SphereLocation(
    MaskSpawnGroup=false
    Override Seed=false
    Sphere Radius=5
    Surface Only=false
    UseNonUniformScale=false
    UseOffset=false
    UseSphereOrientation=false
   )
   &AddVelocity(Velocity=&UniformRangedVector(Minimum=<-20 -5 10>, Maximum=<20 -20 30>)) {enabled=false}
   &InheritVelocity(Module.Limit Speed=true, Velocity Limit=50)
  ]
  particleUpdateScript [
   &ParticleState()
   &GravityForce(Gravity=&RandomRangeVector(Minimum=<0 0 -150>, Maximum=<0 0 150>))
   &Drag(Drag=&UniformRangedFloat(Minimum=2, Maximum=4), Use Linear Drag=true, Use Rotational Drag=false)
   &CurlNoiseForce(Noise Strength=300, Pan Noise=false) {enabled=false}
   &SolveForcesAndVelocity(Clamp Velocity=false, Limit Acceleration=false)
   &ScaleSpriteSize(
    Uniform Curve Scale=&Sine(
     Normalized Angle=&Add_Float(A=$Emitter.Age, B=&RandomRangeFloat(Minimum=0, Maximum=1))
     Period=&RandomRangeFloat(Minimum=0.25, Maximum=1)
     Scale=1
     Bias=0
    )
    Uniform Curve Sprite Scale=DataInterface() {propertyValues {Curve {keys [
        {arriveTangent=5.7198, interpMode=RCIM_Cubic, time=0, value=0}
        {interpMode=RCIM_Cubic, tangentMode=RCTM_User, time=0.2, value=1}
        {interpMode=RCIM_Cubic, tangentMode=RCTM_User, time=0.5, value=1}
        {interpMode=RCIM_Cubic, leaveTangent=-5.5449, time=1, value=0}
       ]}}}
   )
  ]
  rendererClasses [&NiagaraSpriteRendererProperties]
  renderers [{rendererClass=&NiagaraSpriteRendererProperties}]
 }
 Trail {
  emitterUpdateScript [
   &EmitterState()
   &FN_SpawnAnimTrail_v2()
  ]
  particleSpawnScript [
   /Game/Effects/Niagara/Modules/AnimTrail/FN_InitializeAnimTrail/FN_InitializeAnimTrailBoneIndecies(Second Bone or Socket Index=1)
   /Game/Effects/Niagara/Modules/AnimTrail/FN_InitializeAnimTrail/FN_InitializeAnimTrail(AnimTrailWidth=2, Color=<rgba 0.5 0 0.1667 1>, Lifetime=0.1333)
  ]
  particleUpdateScript [
   &FN_UpdateAnimTrail_v2()
   &ScaleColor(
    Scale Alpha=&FloatFromCurve(
     FloatCurve=DataInterface() {propertyValues {Curve {keys [
         {interpMode=RCIM_Cubic, tangentMode=RCTM_User, time=0.05, value=0}
         {interpMode=RCIM_Cubic, tangentMode=RCTM_User, time=0.25, value=1}
         {interpMode=RCIM_Cubic, tangentMode=RCTM_User, time=0.5, value=1}
         {tangentMode=RCTM_User, time=0.9, value=0}
        ]}}}
     CurveIndex=&ReturnNormalizedExecIndex()
    )
    Scale RGB=&Multiply_VectorByFloat(
     Vector=<1 1 1>
     Float=&FNTimeOfDayScalar(Night Scale=0.5, Morning Scale=0.66, Evening Scale=0.66)
    )
    ScaleA=true
    ScaleRGB=true
   )
  ]
  rendererClasses [&NiagaraRibbonRendererProperties]
  renderers [{rendererClass=&NiagaraRibbonRendererProperties}]
 }
}
summary {
 systemName=NS_Pickaxe_FlowerSkeletonLoveMale_1H_Swing
 emitters [
  {emitterName=Petals, rendererClasses [&NiagaraSpriteRendererProperties]}
  {emitterName=Sparkles, rendererClasses [&NiagaraSpriteRendererProperties]}
  {emitterName=Trail, rendererClasses [&NiagaraRibbonRendererProperties]}
 ]
 userVariables [
  User.RibbonIDOffset {type=NiagaraInt32}
  User.SkeletalMesh {
   type=NiagaraDataInterfaceSkeletalMesh
   defaultValue {
    struct=NiagaraExt_VariableValue_DataInterface
    value {dataInterface=&self:NiagaraDataInterfaceSkeletalMesh_0}
   }
  }
 ]
}
